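# Animation Blueprint for the SleepyDuck pickaxe that spins its dyn_Spin bone while a melee animation plays.
# Each update accumulates a roll angle in current_velocity at a rate chosen by bIsPlayingMeleeAnim, through a Lerp with an out-of-range alpha that extrapolates.
# An interpolated float copy of that flag drives the alpha of an additive bone-space Modify Bone over the input pose, easing the spin in and out.
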AnimBlueprint /BRCosmetics/Weapons/FORT_Melee/Pickaxe_SleepyDuck/Meshes/SleepyDuck_Axe_AnimBP
target_skeleton=&/SleepyDuck_Axe_Skeleton
graphs [
 AnimGraph {nodes [
   AnimGraphNode_SubInput_0 = Input Pose()
   n1 = "Transform (Modify) Bone - Bone: dyn_Spin"(
    ComponentPose=Local To Component(LocalPose=@AnimGraphNode_SubInput_0)
    Alpha=Get bIsPlayingMeleeAnim float() -> (bIsPlayingMeleeAnim float:double)
    Translation="0,0,0"
    Rotation=Get Rotator() -> (Rotator:Rotator)
    Scale="1,1,1"
   ) {
    rotationMode=BMM_Additive
    rotationSpace=BCS_BoneSpace
    alphaBoolBlend {blendInTime=0.1, blendOption=Linear, blendOutTime=0.1}
    alphaScaleBiasClamp {bInterpResult=true, interpSpeedDecreasing=5}
    boneToModify {boneName=dyn_Spin}
   }
   Output Pose(Result=Component To Local(ComponentPose=@n1)) {layerGroup=None}
  ]}
 EventGraph {nodes [
   n1 = Event BlueprintUpdateAnimation()
   > Set Delta Time X(Delta Time X=@n1.DeltaTimeX)
   > n2 = Sequence()
   n3 = Get bIsPlayingMeleeAnim()
   Set bIsPlayingMeleeAnim float(
    execute=@n2.then_1
    bIsPlayingMeleeAnim float:double=FInterpTo(Current=Get bIsPlayingMeleeAnim float(), Target="To Float (Boolean)"(InBool=@n3), DeltaTime=Get Delta Time X(), InterpSpeed=8)
   ) -> (Output_Get:double)
   n4 = Lerp(A=Select(Option 0:double=0, Option 1:double=470, Index=Get bIsPlayingMeleeAnim()) -> (ReturnValue:double), B=1, Alpha=20)
   n5 = Get Delta Time X()
   n6 = Set current_velocity(
    execute=@n2.then_0
    current_velocity:double=float + float(A=float * float(A=@n5, B=@n4) {is=title K2Node_PromotableOperator}, B=Get current_velocity() -> (current_velocity:double))
   ) -> (Output_Get:double)
   > Set Rotator(Rotator:Rotator=MakeRotator(Roll=@n6)) -> (Output_Get:Rotator)
  ]}
]
variables [
 bIsPlayingMeleeAnim float {type=number}
 current_velocity {type=number}
 Delta Time X {type=number}
 Rotator {type=Rotator}
]
